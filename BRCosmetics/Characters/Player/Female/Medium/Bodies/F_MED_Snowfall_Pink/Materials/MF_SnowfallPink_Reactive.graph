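# Material function for the Snowfall Pink outfit that adds reactive frost, ice shards and glow to incoming material attributes, behind a static switch.
# Blueprint-driven elimination count and timers grow a frost layer from a skin FX mask, using pre-skinned biplanar projection so it sticks to the body.
# Stage value reveals vertex-color-masked shards through vertex offset and dithered opacity.
# A timed cosine wave adds fading highlights; a FrostMask is also output.
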MaterialFunction /BRCosmetics/Characters/Player/Female/Medium/Bodies/F_MED_Snowfall_Pink/Materials/MF_SnowfallPink_Reactive
graph {
 properties {previewBlendMode=BLEND_Masked}
 nodes [
  n1 = NamedRerouteDeclaration(None=FunctionInput() {inputName=MA, inputType=FunctionInput_MaterialAttributes}) {name=MA_In, nodeColor=<rgba 0.8516 1 0 1>}
  n2 = GetMaterialAttributes(None=@n1) {attributeGetTypes [
    #1
    #2
    #3
    #4
   ]}
  n3 = Constant()
  n4 = NamedRerouteUsage()
  n5 = SetMaterialAttributes!(
   MaterialAttributes=@n2.MaterialAttributes
   BaseColor=NamedRerouteUsage()
   Metallic=@n3
   Specular=Constant() {r=0.5}
   Roughness=ScalarParameter() {defaultValue=0.3, group=SnowfallPink, parameterName=Frost Roughness, sliderMax=1}
   Normal=ShadingPathSwitch(Default=@n4, Deferred=@n4, Mobile=Constant3Vector() {constant=<rgba 0 0 1 0>}) {desc=Migrated from FeatureLevelSwitch}
   AmbientOcclusion=Constant() {r=1}
   Opacity=@n3
  ) {attributeSetTypes [
    #1
    #5
    #6
    #2
    #4
    #7
    #8
   ]}
  n6 = BlendMaterialAttributes(A=@n1, B=@n5, Alpha=NamedRerouteUsage())
  n7 = NamedRerouteUsage()
  n8 = Add(A=NamedRerouteUsage(), B=NamedRerouteUsage())
  n9 = NamedRerouteUsage()
  n10 = ShadingPathSwitch(Default=@n9, Deferred=@n9, Mobile=Constant() {r=1}) {desc=Migrated from FeatureLevelSwitch}
  n11 = StaticBoolParameter() {group=SnowfallPink, parameterName=Use Snowfall Pink Reactivity, sortPriority=0}
  FunctionOutput(None=StaticSwitch(
    True=SetMaterialAttributes!(MaterialAttributes=@n6, WorldPositionOffset=@n7, EmissiveColor=@n8, OpacityMask=@n10) {attributeSetTypes [#9, #3, #10]}
    False=NamedRerouteUsage()
    Value=@n11
   )) {outputName=MA}
  FunctionOutput(None=StaticSwitch(True=NamedRerouteUsage(), False=Constant(), Value=@n11)) {outputName=FrostMask, sortPriority=1}
  n12 = TextureObject() {texture=/Game/Effects/Fort_Effects/Textures/Snowflake/T_LightFrost}
  n13 = ScalarParameter() {defaultValue=12, group=SnowfallPink, parameterName=Frost_Tiling}
  n14 = PreSkinnedPosition()
  n15 = ShaderStageSwitch(PixelShader=VertexInterpolator(VS=@n14).PS, VertexShader=@n14)
  n16 = PreSkinnedNormal()
  n17 = Normalize(VectorInput=ShaderStageSwitch(PixelShader=VertexInterpolator(VS=@n16).PS, VertexShader=@n16))
  MaterialFunctionCall(TextureObject=@n12, TextureSize=@n13, WorldPosition=@n15, World Space Normal=@n17) {materialFunction=&WorldAlignedTexture}
  n18 = MaterialFunctionCall(Float3=NamedRerouteDeclaration(None=MaterialFunctionCall!(
     ColorTex=@n12
     NormalTex
     Pos=@n15
     Nrm=@n17
     ColorUVScale=@n13
     ColorBlendFactor
     NormalStrength
     NormalUVScale
     NormalBlendFactor
     Local Space Inputs
    ) {materialFunction=/Game/Global/MaterialFunctions/MF_BiPlanar}.Color) {name=FrostMapFloat3, nodeColor=<rgba 0 1 1 1>}) {materialFunction=&BreakOutFloat3Components}
  n19 = NamedRerouteDeclaration(None=Max(A=Max(A=@n18.R, B=@n18.R), B=@n18.B)) {name=FrostMap, nodeColor=<rgba 0 1 0.0625 1>}
  n20 = Subtract(A=@n19, B=Constant() {r=1})
  n21 = ScalarParameter() {desc="Dont Touch this parameter is driven by Blueprint!!", group=SnowfallPink, parameterName=SP_Current_Eliminations, sliderMax=6}
  n22 = Subtract(A=@n21)
  n23 = Time()
  n24 = Add(A=ScalarParameter() {defaultValue=2, group=SnowfallPink, parameterName=SP_Anim_Duration_Frost, sliderMin=0}) {constB=0}
  n25 = Subtract(A=Time(), B=ScalarParameter() {defaultValue=-99999, group=SnowfallPink, parameterName=SP_CurrentTime_Frost, sliderMin=-99999})
  n26 = StaticSwitchParameter(True=Frac(None=Divide(A=@n23, B=@n24)), False=Saturate(None=Divide(A=@n25, B=@n24))) {desc="Dont Touch this parameter is driven by Blueprint!!", group=SnowfallPink, parameterName=UseDebugAnimation}
  n27 = Divide(A=Max(B=LinearInterpolate(A=@n22, B=@n21, Alpha=@n26)), B=Constant() {desc=max kills, r=6})
  n28 = Multiply(A=SmoothStep(Min=OneMinus(None=@n27), Value=NamedRerouteUsage()), B=Constant() {r=2})
  n29 = Multiply(A=Saturate(None=Add(A=@n20, B=@n28)), B=ScalarParameter() {defaultValue=1, group=SnowfallPink, parameterName=Frost_Opacity, sliderMax=1})
  n30 = NamedRerouteUsage()
  NamedRerouteDeclaration(None=Saturate(None=Add(A=@n29, B=@n30))) {name=FrostMask_Out, nodeColor=<rgba 0 0.3407 1 1>}
  n31 = TextureSample() {
   samplerType=SAMPLERTYPE_Masks
   texture=/BRCosmetics/Characters/Player/Female/Medium/Bodies/F_MED_Snowfall_Pink/Textures/T_F_MED_SnowfallPink_Body_FX
  }
  n32 = ChannelMaskParameter(None=FunctionInput(Preview=@n31.RGBA) {bUsePreviewValueAsDefault=true, inputName=SkinFX, inputType=FunctionInput_Vector4, sortPriority=1}) {group=SnowfallPink, maskChannel=Green, parameterName=FrostMask Channel, sortPriority=2, defaultValue=<rgba 0 1 0 0>}
  n33 = Power(
   Base=NamedRerouteDeclaration(None=Add(A=@n32.R) {constB=-0.004, desc=add one step}) {name=FrostMask_In, nodeColor=<rgba 0 1 0.8242 1>}
   Exp=ScalarParameter() {defaultValue=1, group=SnowfallPink, parameterName=FrostGradient_Exp, sliderMax=16}
  )
  n34 = ScalarParameter() {group=SnowfallPink, parameterName=FrostGradient_Subtract, sliderMax=1}
  NamedRerouteDeclaration(None=Saturate(None=Subtract(A=@n33, B=@n34))) {name=FrostMask_Tuned, nodeColor=<rgba 0 1 0.4566 1>}
  n35 = Time()
  n36 = Add(A=ScalarParameter() {defaultValue=2, group=SnowfallPink, parameterName=SP_Anim_Duration_Wave, sliderMin=0}) {constB=0}
  n37 = Subtract(A=Time(), B=ScalarParameter() {defaultValue=-99999, group=SnowfallPink, parameterName=SP_CurrentTime_Wave, sliderMin=-99999})
  n38 = StaticSwitchParameter(True=Frac(None=Divide(A=@n35, B=@n36)), False=Saturate(None=Divide(A=@n37, B=@n36))) {defaultValue=true, desc="Dont Touch this parameter is driven by Blueprint!!", group=SnowfallPink, parameterName=UseDebugAnimation}
  NamedRerouteDeclaration(None=@n38) {name=TriggeredBlend_Wave, nodeColor=<rgba 0.2422 0 1 1>}
  n39 = OneMinus(None=NamedRerouteUsage())
  n40 = Subtract(A=NamedRerouteUsage(), B=Constant() {r=1})
  n41 = Add(A=NamedRerouteUsage(), B=NamedRerouteUsage()) {constB=0.5}
  n42 = Constant() {r=0.5}
  n43 = Multiply(A=Cosine(None=Add(A=@n41) {constB=-0.09}), B=@n42)
  n44 = Multiply(A=Power(Base=Add(A=@n43, B=@n42)) {constExponent=12}) {constB=0.8}
  n45 = Constant() {r=2}
  n46 = Saturate(None=Add(A=@n40, B=Multiply(A=@n44, B=@n45)))
  n47 = VertexColor()
  n48 = StaticSwitchParameter(True=OneMinus(None=@n47.A) {desc=body mask}, False=@n47.A) {defaultValue=true, group=SnowfallPink, parameterName=UseInvertedBodyMask}
  NamedRerouteDeclaration(None=Multiply(A=@n39, B=Multiply(A=@n46, B=@n48))) {name=Frost_Wave, nodeColor=<rgba 0 1 0.6016 1>}
  NamedRerouteDeclaration(None=@n2.BaseColor) {name=BC_In, nodeColor=<rgba 1 1 0 1>}
  n49 = Constant3Vector() {constant=<rgba 0 0 1 0>}
  n50 = VertexColor()
  n51 = SphereMask(A=@n50.R, B=Constant() {r=0.1}) {attenuationRadius=0.01}
  n52 = SphereMask(A=@n50.R, B=Constant() {r=0.2}) {attenuationRadius=0.01}
  n53 = Add(A=Add(A=Multiply(A=@n51) {constB=3}, B=Multiply(A=@n52) {constB=2}), B=SphereMask(A=@n50.R, B=Constant() {r=0.3}) {attenuationRadius=0.01})
  n54 = ScalarParameter() {desc="Dont Touch this parameter is driven by Blueprint!!", group=SnowfallPink, parameterName=SP_Current_Stage, sliderMax=3}
  n55 = Add(A=ScalarParameter() {defaultValue=2, group=SnowfallPink, parameterName=SP_Anim_Duration_IceShards, sliderMin=0}) {constB=0}
  n56 = Frac(None=Divide(A=Time(), B=@n55))
  n57 = Time()
  n58 = ScalarParameter() {defaultValue=-99999, group=SnowfallPink, parameterName=SP_CurrentTime_IceShards, sliderMin=-99999}
  n59 = Saturate(None=Divide(A=Subtract(A=@n57, B=@n58), B=@n55))
  n60 = LinearInterpolate(
   A=Subtract(A=@n54)
   B=@n54
   Alpha=NamedRerouteDeclaration(None=StaticSwitchParameter(True=@n56, False=@n59) {desc="Dont Touch this parameter is driven by Blueprint!!", group=SnowfallPink, parameterName=UseDebugAnimation}) {name=TriggeredBlend_IceShards, nodeColor=<rgba 0.9556 0 1 1>}
  )
  n61 = Saturate(None=Add(A=Add(A=@n53) {constB=-3}, B=Max(B=@n60)))
  n62 = VertexColor()
  n63 = Multiply(A=OneMinus(None=@n62.R), B=@n62.A)
  n64 = NamedRerouteDeclaration(None=Saturate(None=Multiply(A=@n63) {constB=2}) {desc=Shards Mask}) {name=IceShardsMask, nodeColor=<rgba 0.6641 0 1 1>}
  n65 = Multiply(
   A=Multiply(A=Transform(None=@n49), B=ScalarParameter() {defaultValue=-2, group=SnowfallPink, parameterName=VertexOffset, sliderMin=-3}) {constB=-2}
   B=Multiply(A=OneMinus(None=@n61), B=@n64)
  )
  NamedRerouteDeclaration(None=@n65) {name=WPO_Out, nodeColor=<rgba 0.3783 0.3917 1 1>}
  n66 = MaterialFunctionCall(Float3=Divide(A=NamedRerouteUsage())) {materialFunction=&BreakOutFloat3Components}
  n67 = Add(A=Add(A=Add(A=@n66.R, B=@n66.R), B=@n66.B)) {constB=0.1}
  n68 = SmoothStep(Min=OneMinus(None=Saturate(None=@n67)), Value=@n61)
  NamedRerouteDeclaration(None=LinearInterpolate(
    A=Constant() {r=1}
    B=MaterialFunctionCall(Alpha Threshold=@n68) {materialFunction=&DitherTemporalAA}.Result
    Alpha=@n64
   )) {name=OpacityMask_Out, nodeColor=<rgba 0.4766 1 0 1>}
  NamedRerouteDeclaration(None=@n2.Roughness) {name=R_In, nodeColor=<rgba 0.5625 0.5625 0.5625 1>}
  n69 = VectorParameter() {group=SnowfallPink, parameterName=FrostColor_Dark, sortPriority=3, defaultValue=<rgba 0.0475 0.3089 0.5365 1>}
  n70 = VectorParameter() {group=SnowfallPink, parameterName=FrostColor_Bright, sortPriority=3, defaultValue=<rgba 0.7652 0.882 0.9479 1>}
  n71 = Multiply(A=@n19, B=@n19)
  NamedRerouteDeclaration(None=LinearInterpolate(A=@n69.RGB, B=@n70.RGB, Alpha=Saturate(None=@n71))) {name=BC_Out, nodeColor=<rgba 1 0.9169 0.2292 1>}
  n72 = NamedRerouteUsage()
  n73 = Fresnel() {baseReflectFraction=0.3, exponent=1}
  n74 = Saturate(None=Power(Base=OneMinus(None=@n73)) {constExponent=8})
  n75 = Constant() {r=0.5}
  n76 = Add(A=Multiply(A=Cosine(None=@n41), B=@n75), B=@n75)
  n77 = Multiply(A=Power(Base=@n76) {constExponent=4}, B=Constant() {r=2})
  n78 = Multiply(A=Saturate(None=Multiply(A=@n72, B=@n74)), B=Saturate(None=Add(A=@n40, B=@n77)))
  n79 = Multiply(A=Power(Base=@n76) {constExponent=24}) {constB=0.5}
  n80 = Constant() {r=2}
  n81 = Saturate(None=Add(A=@n40, B=Multiply(A=@n79, B=@n80)))
  n82 = Multiply(A=Power(Base=Multiply(A=@n81, B=@n48)) {constExponent=6}) {constB=0.6}
  NamedRerouteDeclaration(None=Multiply(A=Add(A=@n78, B=@n82) {constB=0}, B=@n39)) {name=IceShardsGlow, nodeColor=<rgba 0.7812 1 0 1>}
  NamedRerouteDeclaration(None=@n2.EmissiveColor) {name=E_In, nodeColor=<rgba 1 0 0.9324 1>}
  n83 = ScalarParameter() {group=SnowfallPink, parameterName=SP_Current_Stage}
  n84 = Multiply(
   A=TextureSampleParameter2D() {group=Emissive, parameterName=Emissive_Reactive, texture=&T_BlacksRGB}.RGB
   B=Saturate(None=Multiply(A=@n83) {constB=0.5})
  )
  n85 = Constant()
  n86 = VectorParameter() {group=SnowfallPink, parameterName=Glow Color, sortPriority=3, defaultValue=<rgba 0 0.6584 1 5>}
  n87 = Multiply(A=@n86.RGB, B=@n86.RGB)
  n88 = NamedRerouteUsage()
  NamedRerouteDeclaration(None=Add(A=StaticSwitchParameter(True=@n84, False=@n85) {group=SnowfallPink, parameterName=UseReactiveEmissive}, B=Multiply(A=@n87, B=@n88))) {name=E_Out, nodeColor=<rgba 1 0 0.5156 1>}
  NamedRerouteDeclaration(None=@n2.Normal) {name=N_In, nodeColor=<rgba 0 0.4844 1 1>}
  n89 = TextureObject() {samplerType=SAMPLERTYPE_Normal, texture=/Game/Effects/Fort_Effects/Textures/Misc/T_NoiseNormal}
  n90 = ScalarParameter() {defaultValue=50, group=SnowfallPink, parameterName=FrostNormalMap_Tiling, sliderMax=100, sliderMin=1}
  n91 = PreSkinnedPosition()
  n92 = ShaderStageSwitch(PixelShader=VertexInterpolator(VS=@n91).PS, VertexShader=@n91)
  n93 = PreSkinnedNormal()
  n94 = ShaderStageSwitch(PixelShader=VertexInterpolator(VS=@n93).PS, VertexShader=@n93)
  NamedRerouteDeclaration(None=MaterialFunctionCall(
    Normal=MaterialFunctionCall(TextureObject=@n89, TextureSize=@n90, WorldPosition=@n92, World Space Normal=@n94) {materialFunction=&WorldAlignedTexture}.XYZ Texture
    Flatness=ScalarParameter() {defaultValue=0.5, group=SnowfallPink, parameterName=FrostNormal_Intensity, sliderMax=2}
   ) {materialFunction=&FlattenNormal}.Result) {name=N_Out, nodeColor=<rgba 0 0.2892 1 1>}
  ShadingPathSwitch(Default=@n65, Deferred=@n65, Mobile=Constant()) {desc=Migrated from FeatureLevelSwitch}
  n95 = TextureCoordinate()
  n96 = VectorParameter() {group=SnowfallPink, parameterName=Frost_Tiling_Ratio, defaultValue=<rgba 1 1 0 1>}
  TextureSample(UVs=Multiply(A=Multiply(A=@n13, B=@n95), B=ComponentMask(None=@n96.RGB) {g=true, r=true})) {constMipValue=0, texture=&T_LightFrost}
  VertexNormalWS()
 ]
}
