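# Material function that layers a stylized ice effect onto incoming material attributes for the Snowfall Pink character body.
# Surface, noise and facet normals use bump-offset UV or pre-skinned triplanar sampling, chosen by static switch.
# Emissive combines parallax cracks, fake reflections and Fresnel-driven color curves, scaled by a time-of-day curve.
# A channel of the Alpha input masks the blend, gated by the UseBlackIce_Snowfall switch.
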MaterialFunction /BRCosmetics/Characters/Player/Female/Medium/Bodies/F_MED_Snowfall_Pink/Materials/MF_SowfallPink_Ice
graph {nodes [
  n1 = FunctionInput() {inputType=FunctionInput_MaterialAttributes}
  n2 = GetMaterialAttributes(None=@n1) {attributeGetTypes [#1, #2]}
  n3 = TextureObjectParameter() {group=IceBreaker, parameterName=IB_SurfaceNormal, samplerType=SAMPLERTYPE_Normal, texture=/Game/Weapons/WeaponSkins/Textures/StoneNormal}
  n4 = MaterialFunctionCall!(
   TextureObject=@n3
   TextureSize=NamedRerouteUsage()
   WorldPosition=NamedRerouteUsage()
   Export Float 4
   World Space Normal=NamedRerouteUsage()
   ProjectionTransitionContrast
   Bump Height=NamedRerouteUsage()
   Bump Height Ratio=NamedRerouteUsage()
  ) {materialFunction=/Game/Characters/Player/Female/Medium/Bodies/F_MED_JurassicArchaeology/Skins/Winter/Materials/WorldAlignedTexture_WithBumpOffset}
  n5 = TextureSampleParameter2D(UVs=NamedRerouteUsage()) {
   group=IceBreaker
   parameterName=IB_SurfaceNormal
   samplerSource=SSM_Wrap_WorldGroupSettings
   samplerType=SAMPLERTYPE_Normal
   texture=&StoneNormal
  }
  n6 = MaterialFunctionCall(
   Normal=StaticSwitchParameter(True=@n4.XYZ Texture, False=@n5.RGB) {group=IceBreaker, parameterName=Use Triplanar}
   Flatness=ScalarParameter() {defaultValue=-0.5, group=BlackIce, parameterName=ice normal flatness, sliderMax=1}
  ) {materialFunction=&FlattenNormal}
  n7 = TextureSample(UVs=BumpOffset(
    Coordinate=TextureCoordinate() {uTiling=0.66, vTiling=0.66}
    Height=ScalarParameter() {defaultValue=-0.1, group=BlackIce, parameterName=IB_HeightOffset2}
    HeightRatioInput=ScalarParameter() {defaultValue=0.1, group=BlackIce, parameterName=IB_HeightRatio2}
   )) {
   samplerSource=SSM_Wrap_WorldGroupSettings
   samplerType=SAMPLERTYPE_Normal
   texture=/Game/Athena/Apollo/Environments/Landscape/Textures/Roads/Road_Cracks_N
  }
  n8 = OneMinus(None=Saturate(None=Fresnel() {baseReflectFraction=0, exponent=0.15}))
  n9 = TextureObjectParameter() {
   group=IceBreaker
   parameterName=IceNoise
   samplerType=SAMPLERTYPE_Normal
   texture=/Game/Weapons/WeaponSkins/Wraps/Textures/IceBreaker/T_IceNormal
  }
  n10 = ScalarParameter() {group=BlackIce, parameterName=IB_HeightOffset}
  n11 = ScalarParameter() {group=BlackIce, parameterName=IB_HeightRatio}
  n12 = MaterialFunctionCall(
   TextureObject=@n9
   TextureSize=NamedRerouteUsage()
   WorldPosition=NamedRerouteUsage()
   World Space Normal=NamedRerouteUsage()
   Bump Height=@n10
   Bump Height Ratio=@n11
  ) {materialFunction=&WorldAlignedTexture_WithBumpOffset}
  n13 = TextureSampleParameter2D(UVs=BumpOffset(Coordinate=TextureCoordinate(), Height=@n10, HeightRatioInput=@n11)) {
   group=IceBreaker
   parameterName=IceNoise
   samplerSource=SSM_Wrap_WorldGroupSettings
   samplerType=SAMPLERTYPE_Normal
   texture=&T_IceNormal
  }
  n14 = MaterialFunctionCall(Color=StaticSwitchParameter(True=@n12.XYZ Texture, False=@n13.RGB) {group=IceBreaker, parameterName=Use Triplanar}) {materialFunction=&SplitComponents}
  n15 = OneMinus(None=@n14.R)
  n16 = Saturate(None=Multiply(A=Multiply(A=@n7.G, B=@n8), B=Power(Base=@n15)))
  n17 = ScalarParameter() {defaultValue=0.5, group=BlackIce, parameterName=CrackIntensity}
  n18 = TextureObjectParameter() {
   group=IceBreaker
   parameterName=IB_Facet_Normal
   samplerType=SAMPLERTYPE_Normal
   texture=/Game/Weapons/WeaponSkins/Wraps/Textures/IceBreaker/T_IceFacet_Normal
  }
  n19 = MaterialFunctionCall(
   TextureObject=@n18
   TextureSize=NamedRerouteUsage()
   WorldPosition=NamedRerouteUsage()
   World Space Normal=NamedRerouteUsage()
   Bump Height=NamedRerouteUsage()
   Bump Height Ratio=NamedRerouteUsage()
  ) {materialFunction=&WorldAlignedTexture_WithBumpOffset}
  n20 = TextureSampleParameter2D(UVs=NamedRerouteUsage()) {
   group=IceBreaker
   parameterName=IB_Facet_Normal
   samplerSource=SSM_Wrap_WorldGroupSettings
   samplerType=SAMPLERTYPE_Normal
   texture=&T_IceFacet_Normal
  }
  n21 = Transform(None=MaterialFunctionCall(
    Normal=StaticSwitchParameter(True=@n19.XYZ Texture, False=@n20.RGB) {group=IceBreaker, parameterName=Use Triplanar}
    Flatness=ScalarParameter() {parameterName=IB_Facet_Normal_FLattness, sliderMax=1}
   ) {materialFunction=&FlattenNormal}.Result) {transformSourceType=TRANSFORMSOURCE_World, transformType=TRANSFORM_Tangent}
  n22 = TextureSampleParameter2D(UVs=ComponentMask(None=ReflectionVectorWS(CustomWorldNormal=@n21)) {g=true, r=true}) {
   parameterName=" IB_FakeReflectionNoise"
   samplerSource=SSM_Wrap_WorldGroupSettings
   texture=&T_Inky_Smoke_Tile_Chromatic
  }
  n23 = VectorParameter() {group=BlackIce, parameterName=Reflection_Tint, defaultValue=<rgba 0.1 0.1 0.1 1>}
  n24 = ScalarParameter() {defaultValue=0.4, group=IceBreaker, parameterName=IB_IceNormalFlatness, sliderMax=1}
  n25 = Fresnel(
   ExponentIn=ScalarParameter() {defaultValue=2, group=BlackIce, parameterName=IB_FresnelExponent}
   Normal=Transform(None=MaterialFunctionCall(Normal=@n14.RGB, Flatness=@n24) {materialFunction=&FlattenNormal}.Result)
  )
  n26 = CameraVectorWS()
  n27 = Constant3Vector() {constant=<rgba 0 0 1 0>}
  n28 = CurveAtlasRowParameter(CurveTime=Saturate(None=DotProduct(A=@n26, B=@n27))) {
   atlas=&AthenaCurveAtlas01
   curve=/Game/Weapons/WeaponSkins/Wraps/Textures/IceBreaker/IceReflectanceCurve
   defaultValue=198
   group=IceBreaker
   parameterName=IB_IceReflectanceCurve
  }
  n29 = Fresnel(ExponentIn=ScalarParameter() {defaultValue=0.15, group=BlackIce, parameterName=GlowEXP}) {baseReflectFraction=0}
  n30 = ScalarParameter() {defaultValue=0.7864, group=BlackIce, parameterName=CurveAdjust}
  n31 = ComponentMask(None=@n14.RGB) {r=true}
  n32 = Power(Base=DotProduct(A=CameraVectorWS(), B=VertexNormalWS())) {constExponent=12}
  n33 = Max(
   A=Power(Base=@n25, Exp=ScalarParameter() {defaultValue=1.375, group=BlackIce, parameterName=IB_FresnelPower})
   B=Multiply(A=Saturate(None=@n31), B=Saturate(None=@n32))
  )
  n34 = CurveAtlasRowParameter(CurveTime=Saturate(None=Multiply(A=@n30, B=@n33))) {
   atlas=&AthenaCurveAtlas01
   curve=/Game/Weapons/WeaponSkins/Wraps/Textures/IceBreaker/IceBreakerColorCurve
   defaultValue=199
   group=BlackIce
   parameterName=IceBreakerBaseColorCurve
  }
  n35 = Fresnel(ExponentIn=ScalarParameter() {defaultValue=1, group=BlackIce, parameterName=IB_BaseFresnelExponent, sliderMax=1}, Normal=PixelNormalWS())
  n36 = ScalarParameter() {defaultValue=2, group=BlackIce, parameterName=IB_BaseFresnelPower}
  n37 = LinearInterpolate(
   A=Multiply(A=OneMinus(None=@n29), B=@n34)
   B=Multiply(
    A=VectorParameter() {group=BlackIce, parameterName=IB_EdgeColor, defaultValue=<rgba 1 1 1 1>}.RGB
    B=ScalarParameter() {defaultValue=1, group=BlackIce, parameterName=IB_EdgeBrightness}
   )
   Alpha=Saturate(None=Power(Base=@n35, Exp=@n36))
  )
  n38 = Add(
   A=Multiply(A=Multiply(A=@n22.RGB, B=@n23.RGB), B=@n25)
   B=Multiply(A=ScalarParameter() {defaultValue=0.668, group=BlackIce, parameterName=Base Emissive, sliderMax=1}, B=Multiply(A=@n28, B=@n37))
  )
  n39 = ScalarParameter() {defaultValue=2.5, group=BlackIce, parameterName=EmissiveStrength}
  n40 = CurveAtlasRowParameter(CurveTime=CollectionParameter() {collection=&FortniteMaterialParameters, parameterName=TimeOfDay_0-1}) {
   atlas=/Game/Characters/Shared/CurveAtlas/AthenaCharacterCurveAtlas
   curve=/Game/Packages/Fortress_SharedMaterials/Anime_Master_Material/ToonTODBrightness_Curve
   defaultValue=17
   group=TOD_Adjustments
   parameterName=TODBrightnessAdjustmentCurve
  }
  n41 = Multiply(
   A=Add(A=Multiply(A=@n16, B=@n17), B=Multiply(A=@n38, B=@n39))
   B=StaticComponentMaskParameter(None=@n40) {defaultR=true, parameterName=TODCurveChannelMask}
  )
  n42 = ScalarParameter() {group=BlackIce, parameterName=SurfaceNormal_Flatness, sliderMax=1}
  n43 = SetMaterialAttributes!(
   MaterialAttributes=@n1
   BaseColor=Multiply(A=@n2.BaseColor, B=VectorParameter() {group=BlackIce, parameterName=BaseColor, defaultValue=<rgba 0.04 0.04 0.04 1>}.RGB)
   Metallic=ScalarParameter() {defaultValue=0.24, group=BlackIce, parameterName=Metallic, sliderMax=1}
   Specular=ScalarParameter() {defaultValue=0.8871, group=BlackIce, parameterName=spec, sliderMax=1}
   Roughness=LinearInterpolate(
    A=ScalarParameter() {defaultValue=0.1, group=BlackIce, parameterName=Ice_RoughnessMin, sliderMax=1}
    B=ScalarParameter() {defaultValue=0.3, group=BlackIce, parameterName=Ice_RoughnessMax, sliderMax=1}
    Alpha=MaterialFunctionCall(Color=@n6.Result) {materialFunction=&SplitComponents}.G
   )
   EmissiveColor=LinearInterpolate(
    A=@n41
    B=Multiply(A=@n2.BaseColor, B=@n41)
    Alpha=ScalarParameter() {group=BlackIce, parameterName=Tint_Emissive_with_BaseColor, sliderMax=1}
   )
   Normal=MaterialFunctionCall(
    BaseNormal=MaterialFunctionCall(Normal=@n2.Normal, Flatness=@n42) {materialFunction=&FlattenNormal}.Result
    AdditionalNormal=@n6.Result
   ) {materialFunction=&BlendAngleCorrectedNormals}.Result
  ) {attributeSetTypes [
    #2
    #3
    #4
    #5
    #6
    #1
   ]}
  n44 = ChannelMaskParameter(None=FunctionInput() {inputName=Alpha, inputType=FunctionInput_Vector4, sortPriority=1, previewValue=<1 1 1 1>}) {group=Skin FX, maskChannel=Blue, parameterName=BlackIceMaskChannel, defaultValue=<rgba 0 0 1 0>}
  FunctionOutput(None=StaticSwitchParameter(True=BlendMaterialAttributes(A=@n1, B=@n43, Alpha=@n44.R), False=@n1) {group=Skin FX, parameterName=UseBlackIce_Snowfall})
  n45 = TextureCoordinate()
  n46 = ScalarParameter() {defaultValue=1, group=BlackIce, parameterName=IB_Tiling, sliderMax=20}
  n47 = ScalarParameter() {defaultValue=-0.1, group=BlackIce, parameterName=IB_HeightOffset}
  n48 = ScalarParameter() {defaultValue=0.1, group=BlackIce, parameterName=IB_HeightRatio}
  NamedRerouteDeclaration(None=BumpOffset(
    Coordinate=Multiply(A=@n45, B=@n46)
    Height=NamedRerouteDeclaration(None=@n47) {name=IB_HeightOffset, nodeColor=<rgba 0.2656 1 0 1>}
    HeightRatioInput=NamedRerouteDeclaration(None=@n48) {name=IB_HeightRatio, nodeColor=<rgba 0.2656 1 0 1>}
   )) {name=Shatter Bump Offset, nodeColor=<rgba 0 1 0.3438 1>}
  n49 = PreSkinnedPosition()
  NamedRerouteDeclaration(None=ShaderStageSwitch(PixelShader=VertexInterpolator(VS=@n49).PS, VertexShader=@n49)) {name=PreSkinnedLocalPos, nodeColor=<rgba 0.125 0 1 1>}
  n50 = PreSkinnedNormal()
  NamedRerouteDeclaration(None=ShaderStageSwitch(PixelShader=VertexInterpolator(VS=@n50).PS, VertexShader=@n50)) {name=PreSkinnedLocalNormal, nodeColor=<rgba 0 1 0.4844 1>}
  NamedRerouteDeclaration(None=VectorParameter() {group=IceBreaker, parameterName=Size, defaultValue=<rgba 100 100 100 1>}.RGB) {name=TextureSize, nodeColor=<rgba 0 1 0.1797 1>}
  Transform(None=Constant3Vector() {constant=<rgba 0 0 1 0>})
  VertexNormalWS()
 ]}
