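# Character body material for the Snowfall Pink female medium player cosmetic, a masked material with Subsurface Profile skin shading and skeletal mesh support.
# Material attributes flow from the shared character PreFX function through skin-specific Ice and Reactive functions, both fed by the SkinFX mask texture.
# A GlitterFX function borrowed from weapon skins follows, then the shared character PostFX and cinematic PostFX functions finish the chain.
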Material /BRCosmetics/Characters/Player/Female/Medium/Bodies/F_MED_Snowfall_Pink/Materials/M_Snowfall_Pink
graph {
 properties {
  bAutomaticallySetUsageInEditor=false
  bUseMaterialAttributes=true
  bUsedWithMeshDeformer=true
  bUsedWithMorphTargets=true
  bUsedWithSkeletalMesh=true
  blendMode=BLEND_Masked
  pixelDepthOffsetMode=PDOM_Legacy
  shadingModel=MSM_FromMaterialExpression
  subsurfaceProfile=&SS_HeroSkin_02
  usedShadingModels=Default Lit | Subsurface Profile
 }
 nodes [
  n1 = MaterialFunctionCall() {materialFunction=&BaseFN_CharacterFunction_PreFX}
  n2 = TextureObjectParameter() {
   group=Skin FX
   parameterName=SkinFX_Mask
   samplerType=SAMPLERTYPE_Masks
   texture=/BRCosmetics/Characters/Player/Female/Medium/Bodies/F_MED_Snowfall_Pink/Textures/T_F_MED_SnowfallPink_Body_FX
  }
  n3 = TextureSample(Tex=@n2)
  n4 = MaterialFunctionCall!(
   MA=MaterialFunctionCall!(In=@n1.Result, Alpha=@n3.RGBA) {materialFunction=&/MF_SowfallPink_Ice}.Result
   SkinFX=@n3.RGBA
  ) {materialFunction=&/MF_SnowfallPink_Reactive}
  n5 = MaterialFunctionCall(
   AO=@n1.AO
   Distance Blend=@n1.AO
   Fresnel=@n1.AO
   Gmap=@n1.AO
   Pre Skinned Local Position=@n1.AO
   Material=MaterialFunctionCall!(In=@n4.MA, SkinFX_Mask=@n4.MA) {materialFunction=/Game/Weapons/WeaponSkins/MaterialFunction/MF_GlitterFX}.Result
  ) {materialFunction=&BaseFN_CharacterFunction_PostFX}
  MaterialFunctionCall(In=@n5.Result) {materialFunction=&CINE_MF_CharPostFX}
 ]
}
