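# Physics asset for the F_MED_Snowfall_Pink female player body, covering dynamic hair bones and the upper body with capsules.
# Head, neck, upper spine, clavicles and upper arms have angularly free joints, with dyn_hair_MSTR hanging off the head.
# Back, left-B, right-A and right-B hair chains branch from that master bone through base bodies with limited swing and twist.
# Looser cross constraints tie side chains to the back chain and right A to right B.
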PhysicsAsset /BRCosmetics/Characters/Player/Female/Medium/Bodies/F_MED_Snowfall_Pink/Meshes/Parts/F_MED_Snowafall_Pink_Physics
bodies {
 clavicle_l [{shapeType=Capsule, length=9.6806, radius=5.0694, center=<4.4414 -0.8448 -0.7817>, rotation=<rot 79.83 0 0>}]
 clavicle_r [{shapeType=Capsule, length=10.5657, radius=5.6843, center=<-4.5821 0.8448 -0.0078>, rotation=<rot 79.8265 0 180>}]
 dyn_hair_A_a_base_r [{shapeType=Capsule, length=0.505, radius=0.505, shapeName=dyn_hair_A_a_base_r_capsule, rotation=<rot 0 -90 0>}]
 dyn_hair_A_a_r [{shapeType=Capsule, length=3.5, radius=1.75, shapeName=dyn_hair_A_a_r_capsule, center=<3.4132 0 0>, rotation=<rot 83.7407 -180 -180>}]
 dyn_hair_A_b_r [{
   shapeType=Capsule
   length=1.5003
   radius=3.9997
   shapeName=dyn_hair_A_b_r_capsule
   center=<3.1003 0 -1.5>
   rotation=<rot 90 -165.9638 -145.9638>
  }]
 dyn_hair_B_a_base_l [{shapeType=Capsule, length=0.505, radius=0.505, shapeName=dyn_hair_B_a_base_l_capsule, rotation=<rot 0 -90 0>}]
 dyn_hair_B_a_base_r [{shapeType=Capsule, length=0.505, radius=0.505, shapeName=dyn_hair_B_a_base_r_capsule, rotation=<rot 0 -90 0>}]
 dyn_hair_B_a_l [{shapeType=Capsule, length=2.75, radius=1.75, shapeName=dyn_hair_B_a_l_capsule, center=<2.4831 0 0>, rotation=<rot 90 0 0>}]
 dyn_hair_B_a_r [{shapeType=Capsule, length=2.75, radius=1.75, shapeName=dyn_hair_B_a_r_capsule, center=<2.5 0 0>, rotation=<rot 90 0 0>}]
 dyn_hair_B_b_l [{shapeType=Capsule, length=1.25, radius=2.25, shapeName=dyn_hair_B_b_l_capsule, center=<2.5 0 1>, rotation=<rot 90 0 -15>}]
 dyn_hair_B_b_r [{shapeType=Capsule, length=0.75, radius=2.75, shapeName=dyn_hair_B_b_r_capsule, center=<2.1794 0 0>, rotation=<rot 90 0 15>}]
 dyn_hair_B_c_l [{shapeType=Capsule, length=3.75, radius=2.25, shapeName=dyn_hair_B_c_l_capsule, center=<1.4646 0 3>, rotation=<rot 3.8281 3.219 40.1077>}]
 dyn_hair_B_c_r [{shapeType=Capsule, length=8.0386, radius=3.296, shapeName=dyn_hair_B_c_r_capsule, center=<2.7029 -1 -3>, rotation=<rot 10 0 -20>}]
 dyn_hair_B_d_l [{shapeType=Capsule, length=4, radius=2.25, shapeName=dyn_hair_B_d_l_capsule, center=<1.6357 4 2>, rotation=<rot 0 0 90>}]
 dyn_hair_B_d_r [{shapeType=Capsule, length=6.75, radius=3.75, shapeName=dyn_hair_B_d_r_capsule, center=<3 -3 -2.5>, rotation=<rot 0 90 0>}]
 dyn_hair_B_e_l [{shapeType=Capsule, length=3, radius=2.75, shapeName=dyn_hair_B_e_l_capsule, center=<2 2 3>, rotation=<rot -0.0002 0 80.0001>}]
 dyn_hair_B_e_r [{
   shapeType=Capsule
   length=4.7573
   radius=3.039
   shapeName=dyn_hair_B_e_r_capsule
   center=<3 -2 -1.5>
   rotation=<rot -4.9339 107.17 -170.6551>
  }]
 dyn_hair_MSTR [{shapeType=Capsule, length=0.505, radius=0.505, shapeName=dyn_hair_MSTR_capsule}]
 dyn_hair_bk_a [{shapeType=Capsule, length=2.25, radius=2.25, shapeName=dyn_hair_bk_a_capsule, center=<2.92 0 0>, rotation=<rot 90 0 0>}]
 dyn_hair_bk_a_base [{shapeType=Capsule, length=0.505, radius=0.505, shapeName=dyn_hair_bk_a_base_capsule, rotation=<rot 0 -90 0>}]
 dyn_hair_bk_b [{shapeType=Capsule, length=1.25, radius=3, shapeName=dyn_hair_bk_b_capsule, center=<3.3296 0 0>, rotation=<rot 90 0 0>}]
 dyn_hair_bk_c [{shapeType=Capsule, length=7.5, radius=3.25, shapeName=dyn_hair_bk_c_capsule, center=<3 -3 0>}]
 dyn_hair_bk_d [{shapeType=Capsule, length=10.7767, radius=4.2233, shapeName=dyn_hair_bk_d_capsule, center=<3.4798 -4 0>}]
 dyn_hair_bk_e [{shapeType=Capsule, length=11.7767, radius=5.7233, shapeName=dyn_hair_bk_e_capsule, center=<5.5 -4.5 0>}]
 head [
  {shapeType=Capsule, length=5.6906, radius=5.8094, center=<1.9741 2.0294 0>, rotation=<rot 90 0 0>}
  {shapeType=Capsule, length=3, radius=4, center=<-1.887 5.1833 0>, rotation=<rot -90 0.0198 -0.0198>}
  {shapeType=Capsule, length=4.9274, radius=4.0394, center=<-2.2283 4.0356 0.2346>}
 ]
 neck_01 [{shapeType=Capsule, length=4.75, radius=3.75, center=<1.7546 0.8621 0>, rotation=<rot 90 0 0>}]
 spine_04 [
  {shapeType=Capsule, length=2.8024, radius=10.288, center=<4.7773 2.7504 0>}
  {shapeType=Capsule, length=5.962, radius=9.788, center=<-0.6956 2.3489 0>}
 ]
 spine_05 [{shapeType=Capsule, length=8.7756, radius=7.8148, center=<3.0216 0.1354 0>}]
 upperarm_l [{shapeType=Capsule, length=13, radius=4.5, center=<8.9056 -0.0001 0>, rotation=<rot 90 0 0>}]
 upperarm_r [{shapeType=Capsule, length=13, radius=4.5, center=<-8.9056 0.0001 0>, rotation=<rot 90 0 180>}]
}
constraints [
 {
  bone1Name=head
  bone2Name=neck_01
  swing1LimitDegrees=45
  swing1Motion=Free
  swing2LimitDegrees=45
  swing2Motion=Free
  twistLimitDegrees=45
  twistMotion=Free
 }
 {
  bone1Name=neck_01
  bone2Name=spine_05
  swing1LimitDegrees=45
  swing1Motion=Free
  swing2LimitDegrees=45
  swing2Motion=Free
  twistLimitDegrees=45
  twistMotion=Free
 }
 {
  bone1Name=spine_05
  bone2Name=spine_04
  swing1LimitDegrees=45
  swing1Motion=Free
  swing2LimitDegrees=45
  swing2Motion=Free
  twistLimitDegrees=45
  twistMotion=Free
 }
 {
  bone1Name=clavicle_l
  bone2Name=spine_05
  swing1LimitDegrees=45
  swing1Motion=Free
  swing2LimitDegrees=45
  swing2Motion=Free
  twistLimitDegrees=45
  twistMotion=Free
 }
 {
  bone1Name=upperarm_l
  bone2Name=clavicle_l
  swing1LimitDegrees=45
  swing1Motion=Free
  swing2LimitDegrees=45
  swing2Motion=Free
  twistLimitDegrees=45
  twistMotion=Free
 }
 {
  bone1Name=clavicle_r
  bone2Name=spine_05
  swing1LimitDegrees=45
  swing1Motion=Free
  swing2LimitDegrees=45
  swing2Motion=Free
  twistLimitDegrees=45
  twistMotion=Free
 }
 {
  bone1Name=upperarm_r
  bone2Name=clavicle_r
  swing1LimitDegrees=45
  swing1Motion=Free
  swing2LimitDegrees=45
  swing2Motion=Free
  twistLimitDegrees=45
  twistMotion=Free
 }
 {
  bone1Name=dyn_hair_MSTR
  bone2Name=head
  swing1LimitDegrees=45
  swing1Motion=Free
  swing2LimitDegrees=45
  swing2Motion=Free
  twistLimitDegrees=45
  twistMotion=Free
 }
 {
  bone1Name=dyn_hair_bk_a_base
  bone2Name=dyn_hair_MSTR
  swing1LimitDegrees=45
  swing1Motion=Free
  swing2LimitDegrees=45
  swing2Motion=Free
  twistLimitDegrees=45
  twistMotion=Free
 }
 {bone1Name=dyn_hair_bk_a, bone2Name=dyn_hair_bk_a_base, swing1LimitDegrees=30, swing2LimitDegrees=15, twistLimitDegrees=15}
 {bone1Name=dyn_hair_bk_b, bone2Name=dyn_hair_bk_a, swing1LimitDegrees=30, swing2LimitDegrees=15, twistLimitDegrees=15}
 {bone1Name=dyn_hair_bk_c, bone2Name=dyn_hair_bk_b, swing1LimitDegrees=30, swing2LimitDegrees=15, twistLimitDegrees=15}
 {bone1Name=dyn_hair_bk_d, bone2Name=dyn_hair_bk_c, swing1LimitDegrees=30, swing2LimitDegrees=15, twistLimitDegrees=15}
 {bone1Name=dyn_hair_bk_e, bone2Name=dyn_hair_bk_d, swing1LimitDegrees=30, swing2LimitDegrees=15, twistLimitDegrees=15}
 {
  bone1Name=dyn_hair_A_a_base_r
  bone2Name=dyn_hair_MSTR
  swing1LimitDegrees=45
  swing1Motion=Free
  swing2LimitDegrees=45
  swing2Motion=Free
  twistLimitDegrees=45
  twistMotion=Free
 }
 {bone1Name=dyn_hair_A_a_r, bone2Name=dyn_hair_A_a_base_r, swing1LimitDegrees=35, swing2LimitDegrees=35, twistLimitDegrees=15}
 {bone1Name=dyn_hair_A_b_r, bone2Name=dyn_hair_A_a_r, swing1LimitDegrees=35, swing2LimitDegrees=35, twistLimitDegrees=15}
 {
  bone1Name=dyn_hair_B_a_base_l
  bone2Name=dyn_hair_MSTR
  swing1LimitDegrees=45
  swing1Motion=Free
  swing2LimitDegrees=45
  swing2Motion=Free
  twistLimitDegrees=45
  twistMotion=Free
 }
 {bone1Name=dyn_hair_B_a_l, bone2Name=dyn_hair_B_a_base_l, swing1LimitDegrees=20, swing2LimitDegrees=20, twistLimitDegrees=15}
 {bone1Name=dyn_hair_B_b_l, bone2Name=dyn_hair_B_a_l, swing1LimitDegrees=20, swing2LimitDegrees=20, twistLimitDegrees=15}
 {bone1Name=dyn_hair_B_c_l, bone2Name=dyn_hair_B_b_l, swing1LimitDegrees=20, swing2LimitDegrees=20, twistLimitDegrees=15}
 {bone1Name=dyn_hair_B_d_l, bone2Name=dyn_hair_B_c_l, swing1LimitDegrees=20, swing2LimitDegrees=20, twistLimitDegrees=15}
 {bone1Name=dyn_hair_B_e_l, bone2Name=dyn_hair_B_d_l, swing1LimitDegrees=20, swing2LimitDegrees=20, twistLimitDegrees=15}
 {bone1Name=dyn_hair_B_b_l, bone2Name=dyn_hair_bk_b, swing1LimitDegrees=45, swing2LimitDegrees=45, twistLimitDegrees=45}
 {bone1Name=dyn_hair_B_c_l, bone2Name=dyn_hair_bk_b, swing1LimitDegrees=45, swing2LimitDegrees=45, twistLimitDegrees=45}
 {bone1Name=dyn_hair_B_d_l, bone2Name=dyn_hair_bk_c, swing1LimitDegrees=45, swing2LimitDegrees=45, twistLimitDegrees=45}
 {bone1Name=dyn_hair_B_d_r, bone2Name=dyn_hair_B_c_r, swing1LimitDegrees=35, swing2LimitDegrees=35, twistLimitDegrees=15}
 {
  bone1Name=dyn_hair_B_a_base_r
  bone2Name=dyn_hair_MSTR
  swing1LimitDegrees=45
  swing1Motion=Free
  swing2LimitDegrees=45
  swing2Motion=Free
  twistLimitDegrees=45
  twistMotion=Free
 }
 {bone1Name=dyn_hair_B_a_r, bone2Name=dyn_hair_B_a_base_r, swing1LimitDegrees=35, swing2LimitDegrees=35, twistLimitDegrees=15}
 {bone1Name=dyn_hair_B_b_r, bone2Name=dyn_hair_B_a_r, swing1LimitDegrees=35, swing2LimitDegrees=35, twistLimitDegrees=15}
 {bone1Name=dyn_hair_B_c_r, bone2Name=dyn_hair_B_b_r, swing1LimitDegrees=35, swing2LimitDegrees=35, twistLimitDegrees=15}
 {bone1Name=dyn_hair_B_b_r, bone2Name=dyn_hair_bk_b, swing1LimitDegrees=45, swing2LimitDegrees=45, twistLimitDegrees=45}
 {bone1Name=dyn_hair_B_d_r, bone2Name=dyn_hair_bk_c, swing1LimitDegrees=45, swing2LimitDegrees=45, twistLimitDegrees=45}
 {bone1Name=dyn_hair_B_c_r, bone2Name=dyn_hair_bk_b, swing1LimitDegrees=45, swing2LimitDegrees=45, twistLimitDegrees=45}
 {bone1Name=dyn_hair_A_b_r, bone2Name=dyn_hair_B_b_r, swing1LimitDegrees=45, swing2LimitDegrees=45, twistLimitDegrees=45}
 {bone1Name=dyn_hair_B_e_r, bone2Name=dyn_hair_B_d_r, swing1LimitDegrees=35, swing2LimitDegrees=35, twistLimitDegrees=15}
]
