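# Animation Blueprint for the Snowfall Pink female face accessory, driving its dynamic hair.
# Copies parent and body-mesh poses, layered at the upper arms, aims hair look-at bones, then caches the result.
# Head yaw and right-arm rotation relative to spine_05 drive additive hair-base correctives, blended in by bRigidBodyEnabled.
# A RigidBody node then simulates in spine_05 space with state-based gravity override; several stray bone nodes are unconnected.
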AnimBlueprint /BRCosmetics/Characters/Player/Female/Medium/Bodies/F_MED_Snowfall_Pink/Meshes/Parts/F_MED_Snowfall_Pink_FaceAcc_AnimBP
target_skeleton=&/F_MED_Snowfall_Pink_FaceAcc_Skeleton
graphs [
 AnimGraph {nodes [
   "Transform (Modify) Bone - Bone: dyn_hair_B_a_base_l"(Alpha=1, Translation="0,0,1", Rotation="0,0,0", Scale="1,1,1") {
    rotationMode=BMM_Additive
    rotationSpace=BCS_BoneSpace
    scaleMode=BMM_Additive
    scaleSpace=BCS_BoneSpace
    translationMode=BMM_Additive
    translationSpace=BCS_BoneSpace
    alphaBoolBlend {blendOption=Linear}
    alphaScaleBiasClamp {inRange {max=15}}
    boneToModify {boneName=dyn_hair_B_a_base_l}
    translation {z=1}
   }
   n1 = Get headUp() -> (headUp:double)
   n2 = "Transform (Modify) Bone - Bone: per1_hair_B_a_base_l"(Alpha=@n1, Translation="0,-2,0", Rotation="0,-10,0", Scale="1,1,1") {
    rotationMode=BMM_Additive
    rotationSpace=BCS_BoneSpace
    scaleMode=BMM_Additive
    scaleSpace=BCS_BoneSpace
    translationMode=BMM_Additive
    translationSpace=BCS_BoneSpace
    alphaBoolBlend {blendOption=Linear}
    alphaScaleBiasClamp {bMapRange=true, inRange {max=15}}
    boneToModify {boneName=per1_hair_B_a_base_l}
    rotation {yaw=-10}
    translation {y=-2}
   }
   "Transform (Modify) Bone - Bone: per1_hair_B_a_base_r"(ComponentPose=@n2, Alpha=@n1, Translation="0,-2,0", Rotation="0.000,-10,0.000", Scale="1,1,1") {
    rotationMode=BMM_Additive
    rotationSpace=BCS_BoneSpace
    scaleMode=BMM_Additive
    scaleSpace=BCS_BoneSpace
    translationMode=BMM_Additive
    translationSpace=BCS_BoneSpace
    alphaBoolBlend {blendOption=Linear}
    alphaScaleBiasClamp {bMapRange=true, inRange {max=15}}
    boneToModify {boneName=per1_hair_B_a_base_r}
    rotation {yaw=-10}
    translation {y=-2}
   }
   n3 = Get headUp()
   "Transform (Modify) Bone - Bone: dyn_hair_A_a_base_r"(Alpha=@n3, Translation="0,0,0", Rotation="0,-30,0", Scale="1,1,1") {
    rotationMode=BMM_Additive
    rotationSpace=BCS_BoneSpace
    scaleMode=BMM_Additive
    scaleSpace=BCS_BoneSpace
    translationMode=BMM_Additive
    translationSpace=BCS_BoneSpace
    alphaBoolBlend {blendOption=Linear}
    alphaScaleBiasClamp {bMapRange=true, inRange {max=0, min=-50}, outRange {max=0, min=1}}
    boneToModify {boneName=dyn_hair_A_a_base_r}
    rotation {yaw=-30}
   }
   n4 = Get headUp()
   n5 = "Transform (Modify) Bone - Bone: dyn_hair_bk_a_base"(ComponentPose=Local To Component(LocalPose=Use cached pose 'prerigid'()), Alpha=@n4, Translation="0,-3,0", Rotation="0,-25,0", Scale="1,1,1") {
    rotationMode=BMM_Additive
    rotationSpace=BCS_BoneSpace
    translationMode=BMM_Additive
    translationSpace=BCS_BoneSpace
    alphaBoolBlend {blendOption=Linear}
    alphaScaleBiasClamp {bMapRange=true, inRange {max=0, min=-50}, outRange {max=0, min=1}}
    boneToModify {boneName=dyn_hair_bk_a_base}
    rotation {yaw=-25}
    translation {y=-3}
   }
   n6 = "Transform (Modify) Bone - Bone: dyn_hair_B_a_base_l"(ComponentPose=@n5, Alpha=@n4, Translation="0,-4,2", Rotation="5,-25,0", Scale="0.7,1,1") {
    rotationMode=BMM_Additive
    rotationSpace=BCS_BoneSpace
    scaleMode=BMM_Additive
    scaleSpace=BCS_BoneSpace
    translationMode=BMM_Additive
    translationSpace=BCS_BoneSpace
    alphaBoolBlend {blendOption=Linear}
    alphaScaleBiasClamp {bMapRange=true, inRange {max=0, min=-50}, outRange {max=0, min=1}}
    boneToModify {boneName=dyn_hair_B_a_base_l}
    rotation {pitch=5, yaw=-25}
    scale {x=0.7}
    translation {y=-4, z=2}
   }
   n7 = "Transform (Modify) Bone - Bone: dyn_hair_B_a_base_r"(ComponentPose=@n6, Alpha=@n3, Translation="0,-1,-2", Rotation="5,-40,0", Scale="0.7,1,1") {
    rotationMode=BMM_Additive
    rotationSpace=BCS_BoneSpace
    scaleMode=BMM_Additive
    scaleSpace=BCS_BoneSpace
    translationMode=BMM_Additive
    translationSpace=BCS_BoneSpace
    alphaBoolBlend {blendOption=Linear}
    alphaScaleBiasClamp {bMapRange=true, inRange {max=0, min=-50}, outRange {max=0, min=1}}
    boneToModify {boneName=dyn_hair_B_a_base_r}
    rotation {pitch=5, yaw=-40}
    scale {x=0.7}
    translation {y=-1, z=-2}
   }
   n8 = Get headUp()
   n9 = "Transform (Modify) Bone - Bone: dyn_hair_bk_a_base"(ComponentPose=@n7, Alpha=@n8, Translation="0,0,0", Rotation="0,10,0", Scale="1,1,1") {
    rotationMode=BMM_Additive
    rotationSpace=BCS_BoneSpace
    translationMode=BMM_Additive
    translationSpace=BCS_BoneSpace
    alphaBoolBlend {blendOption=Linear}
    alphaScaleBiasClamp {bMapRange=true, inRange {max=15}}
    boneToModify {boneName=dyn_hair_bk_a_base}
    rotation {yaw=10}
   }
   n10 = "Transform (Modify) Bone - Bone: dyn_hair_B_a_base_l"(ComponentPose=@n9, Alpha=@n8, Translation="0,0,0", Rotation="0,15,0", Scale="1,1,1") {
    rotationMode=BMM_Additive
    rotationSpace=BCS_BoneSpace
    scaleMode=BMM_Additive
    scaleSpace=BCS_BoneSpace
    translationMode=BMM_Additive
    translationSpace=BCS_BoneSpace
    alphaBoolBlend {blendOption=Linear}
    alphaScaleBiasClamp {bMapRange=true, inRange {max=15}}
    boneToModify {boneName=dyn_hair_B_a_base_l}
    rotation {yaw=15}
   }
   n11 = "Transform (Modify) Bone - Bone: dyn_hair_B_a_base_r"(ComponentPose=@n10, Alpha=Get headUp(), Translation="0,0,0", Rotation="0,15,0", Scale="1,1,1") {
    rotationMode=BMM_Additive
    rotationSpace=BCS_BoneSpace
    scaleMode=BMM_Additive
    scaleSpace=BCS_BoneSpace
    translationMode=BMM_Additive
    translationSpace=BCS_BoneSpace
    alphaBoolBlend {blendOption=Linear}
    alphaScaleBiasClamp {bMapRange=true, inRange {max=15}}
    boneToModify {boneName=dyn_hair_B_a_base_r}
    rotation {yaw=15}
   }
   n12 = Get Arm R rot() -> (Arm R rot:double)
   n13 = "Transform (Modify) Bone - Bone: dyn_hair_A_a_base_r"(ComponentPose=@n11, Alpha=@n12, Translation="0,0,0", Rotation="-60,-15.000,15.000", Scale="1,1,1") {
    rotationMode=BMM_Additive
    rotationSpace=BCS_BoneSpace
    scaleMode=BMM_Additive
    scaleSpace=BCS_BoneSpace
    translationMode=BMM_Additive
    translationSpace=BCS_BoneSpace
    alphaBoolBlend {blendOption=Linear}
    alphaScaleBiasClamp {inRange {max=15}}
    boneToModify {boneName=dyn_hair_A_a_base_r}
    rotation=<rot -60 -15 15>
   }
   n14 = "Transform (Modify) Bone - Bone: dyn_hair_B_a_base_r"(ComponentPose=@n13, Alpha=@n12, Translation="0,0,0", Rotation="-5,-15,0", Scale="1,1,1") {
    rotationMode=BMM_Additive
    rotationSpace=BCS_BoneSpace
    scaleMode=BMM_Additive
    scaleSpace=BCS_BoneSpace
    translationMode=BMM_Additive
    translationSpace=BCS_BoneSpace
    alphaBoolBlend {blendOption=Linear}
    alphaScaleBiasClamp {inRange {max=15}}
    boneToModify {boneName=dyn_hair_B_a_base_r}
    rotation {pitch=-5, yaw=-15}
   }
   n15 = Component To Local(ComponentPose=@n14)
   n16 = Use cached pose 'prerigid'()
   n17 = Get bRigidBodyEnabled()
   n18 = RigidBody!(
    ComponentPose=Local To Component(LocalPose=Blend Poses by bool(BlendPose_0=@n15, BlendPose_1=@n16, bActiveValue=@n17) {blendType=Linear})
    Alpha
    OverrideWorldGravity:Vector=Get Gravity Override()
    ExternalForce
   ) -> (Pose) {
    bClampLinearTranslationLimitToRefPose=true
    bOverrideWorldGravity=true
    overridePhysicsAsset=&/F_MED_Snowafall_Pink_Physics
    simulationSpace=BaseBoneSpace
    alphaBoolBlend {blendOption=Linear}
    baseBoneRef {boneName=spine_05}
    componentAppliedLinearAccClamp=<1000 1000 1000>
    componentLinearAccScale=<0.32 0.32 0.32>
    componentLinearVelScale=<0.32 0.32 0.32>
    overrideWorldGravity {z=-980}
   }
   Output Pose(Result=Component To Local(ComponentPose=@n18)) {layerGroup=None}
   n19 = Copy Pose From Mesh(SourceMeshComponent=Get Body_Mesh())
   n20 = Copy Pose From Mesh() {bUseAttachedParent=true}
   n21 = "Transform (Modify) Bone - Bone: dyn_hair_lookat_loc"(
    ComponentPose=Local To Component(LocalPose=Layered blend per bone(BasePose=@n20, BlendPoses_0=@n19) {layerSetup [{branchFilters [{boneName=upperarm_l}, {boneName=upperarm_r}]}]})
    Alpha=1
    Translation="15,0,0.000"
    Rotation="0,0,0"
    Scale="1,1,1"
   ) {
    translationMode=BMM_Additive
    translationSpace=BCS_BoneSpace
    alphaBoolBlend {blendOption=Linear}
    boneToModify {boneName=dyn_hair_lookat_loc}
    translation {x=15}
   }
   n22 = Look At - Bone: dyn_hair_bk_a_look_at(ComponentPose=@n21, Alpha=1) {
    alphaBoolBlend {blendOption=Linear}
    boneToModify {boneName=dyn_hair_bk_a_look_at}
    lookAtLocation {x=0}
    lookAtTarget {boneReference {boneName=dyn_hair_lookat_loc}}
    lookAt_Axis {axis=<1 0>}
   }
   n23 = Look At - Bone: dyn_hair_A_a_look_at_r(ComponentPose=@n22, Alpha=1) {
    alphaBoolBlend {blendOption=Linear}
    boneToModify {boneName=dyn_hair_A_a_look_at_r}
    lookAtLocation {x=0}
    lookAtTarget {boneReference {boneName=dyn_hair_lookat_loc}}
    lookAt_Axis {axis=<1 0>}
   }
   n24 = Look At - Bone: dyn_hair_B_a_look_at_r(ComponentPose=@n23, Alpha=1) {
    alphaBoolBlend {blendOption=Linear}
    boneToModify {boneName=dyn_hair_B_a_look_at_r}
    lookAtLocation=<0 -2.5745 6.4156>
    lookAtTarget {boneReference {boneName=dyn_hair_lookat_loc}}
    lookAt_Axis {axis=<1 0>}
   }
   n25 = Look At - Bone: dyn_hair_B_a_look_at_l(ComponentPose=@n24, Alpha=1) {
    alphaBoolBlend {blendOption=Linear}
    boneToModify {boneName=dyn_hair_B_a_look_at_l}
    lookAtLocation {x=0}
    lookAtTarget {boneReference {boneName=dyn_hair_lookat_loc}}
    lookAt_Axis {axis=<1 0>}
   }
   Save cached pose 'prerigid'(Pose=Component To Local(ComponentPose=@n25)) {cacheName=prerigid}
  ]}
 EventGraph {nodes [
   Event BlueprintUpdateAnimation()
   > n1 = Sequence!(execute) -> (then_0, then_1, then_2:Exec)
   Set Arm R rot(
    execute=@n1.then_1
    Arm R rot:double=GetAlphaFromReferenceTransformDelta(SkeletalMeshComponent=Get Body_Mesh(), BoneName=upperarm_r, BaseName=spine_05, InputRange=(X=60,Y=120), TwistAxis="1,0,0")
   ) -> (Output_Get:double)
   n2 = Gravity Override State Function DEPRECATED(execute=@n1.then_2, CharacterPartAnimInstance=Self-Reference(), GameTimeInSeconds=0, GravityOverrideParams=Get Gravity Override Params())
   > Set Gravity Override(Gravity Override=@n2)
   n3 = GetDeltaTransformFromRefPose(self=GetOwnerSkeletalMeshComponent(), BoneName=head, BaseName=spine_05)
   Set headUp(execute=[@n6.else, @n1.then_0], headUp:double=BreakRotator(InRot=BreakTransform(InTransform=@n3).Rotation).Yaw) -> (Output_Get:double)
   Event BlueprintInitializeAnimation()
   > Set Body_Mesh(Body_Mesh=GetOwnerSkeletalMeshComponent())
   > n4 = Cast To PlayerPawn_Athena(Object=TryGetPawnOwner())
   > n5 = Set Pawn Owner(Pawn Owner:Player Pawn Athena Object Reference=@n4) -> (Output_Get:Player Pawn Athena Object Reference)
   > Set Head_Mesh(Head_Mesh:Skeletal Mesh Component Object Reference=GetSkeletalMeshForPartType(self=@n5, PartType=Head)) -> (Output_Get:Skeletal Mesh Component Object Reference)
   n6 = Branch(Condition=Get bIsSkyDiving())
   > Set headUp(headUp=0)
  ]}
]
variables [
 Arm R rot {type=number}
 Body_Mesh {type=SkeletalMeshComponent}
 Gravity Override {type=Vector}
 Gravity Override Params {
  type=GravityOverrideParamsStruct
  default {skydive_FloatingDown {gravityOverride=<0 -700 -270>}, skydive_GoingUp {gravityOverride=<0 -300 -270>}}
 }
 Head_Mesh {type=SkeletalMeshComponent}
 headUp {type=number}
 Pawn Owner {type=PlayerPawn_Athena_C}
]
